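; Looping gameplay cue for the Ascender ascent that adds repeated camera shake for the riding player.
; Looping start caches the target Fort pawn and starts a repeating timer that, for local player controllers, plays the LineAttach1 shake in camera-local space.
; Shake scale is one minus raw speed with a small floor, so it stays at that minimum unless nearly stationary.
; Removal is empty and never explicitly clears the timer.
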(event EventOnApplication(EitherVFX-preferthisone)
  (params (MyTarget :type Actor%20Object%20Reference) (Parameters :type Gameplay%20Cue%20Parameters%20Structure%20%28by%20ref%29) (ParticleComponents :type Array%20of%20FXSystem%20Component%20Object%20References) (AudioComponents :type Array%20of%20Audio%20Component%20Object%20References) (BurstCameraShakeInstance :type Camera%20Shake%20Base%20Object%20Reference) (BurstDecalInstance :type Decal%20Actor%20Object%20Reference))
  (|CallParentFunction MyTarget Parameters ParticleComponents AudioComponents BurstCameraShakeInstance BurstDecalInstance))

(event EventOnRemoval(EitherVFX-preferthisversion)
  (params (MyTarget :type Actor%20Object%20Reference) (Parameters :type Gameplay%20Cue%20Parameters%20Structure%20%28by%20ref%29) (ParticleComponents :type Array%20of%20FXSystem%20Component%20Object%20References) (AudioComponents :type Array%20of%20Audio%20Component%20Object%20References) (BurstCameraShakeInstance :type Camera%20Shake%20Base%20Object%20Reference) (BurstDecalInstance :type Decal%20Actor%20Object%20Reference)))

(event EventOnLoopingStart(EitherVFX-preferthisversion)
  (params (MyTarget :type Actor%20Object%20Reference) (Parameters :type Gameplay%20Cue%20Parameters%20Structure%20%28by%20ref%29) (ParticleComponents :type Array%20of%20FXSystem%20Component%20Object%20References) (AudioComponents :type Array%20of%20Audio%20Component%20Object%20References))
  (bind _asfort_player_pawn (Utilities|Casting|CastToFortPlayerPawn MyTarget)
    (:then
      (Variables|Default|SetPlayerPawn _asfort_player_pawn)
      (Utilities|Time|SetTimerbyEvent (delegate Custom|cameraShakeTimer) 0.15 true))
    (:CastFailed
      (Variables|Default|SetPlayerPawn 0) ; Clear if not valid
    )))

(event Custom|cameraShakeTimer
  (bind _playerpawn (Variables|Default|GetPlayerPawn))
  (bind _returnvalue (Pawn|GetController _playerpawn))
  (Utilities|IsValid _playerpawn
    (:Is%20Valid
      (bind _asplayer_controller (Utilities|Casting|CastToPlayerController _returnvalue))
      (if (Pawn|IsLocalPlayerController _returnvalue)
        (Utilities|IsValid (Variables|/Script/Engine.PlayerController|GetPlayerCameraManager _asplayer_controller)
          (:Is%20Valid
            (bind _returnvalue_2 (CameraShakes|StartCameraShake :self (Variables|/Script/Engine.PlayerController|GetPlayerCameraManager _asplayer_controller) :ShakeClass "/Ascender/Gameplay/Ascender/CameraShake_Ascender_LineAttach1.CameraShake_Ascender_LineAttach1_C" :Scale (Math|Float|Max(Float) (- 1.0 (Math|Vector|VectorLength (Transformation|GetVelocity _playerpawn))) 0.05) :PlaySpace "CameraLocal"))
            (Variables|Default|SetmatineeCamShake (CameraShakes|ConvLegacyCameraShake _returnvalue_2))))))))

(fn ConstructionScript)
