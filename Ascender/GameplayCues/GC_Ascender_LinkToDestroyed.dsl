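; Burst gameplay cue for the Ascender zipline's destruction effects, judging by its name and the sound it plays.
; On burst, it plays the Ascender destroyed sound at the target actor's location.
; It casts the target to B_Athena_Zipline_Ascender and sets the second spawned particle component's world rotation from that actor's Top component, combined with a half turn about Top's up vector.
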(fn OnBurst(EitherVFX-preferthisversion)
  (params (MyTarget :type Actor%20Object%20Reference) (Parameters :type Gameplay%20Cue%20Parameters%20Structure%20%28by%20ref%29) (ParticleComponents :type Array%20of%20FXSystem%20Component%20Object%20References) (AudioComponents :type Array%20of%20Audio%20Component%20Object%20References) (BurstCameraShakeInstance :type Camera%20Shake%20Base%20Object%20Reference) (BurstDecalInstance :type Decal%20Actor%20Object%20Reference))
  (Audio|PlaySoundatLocation MyTarget "/Ascender/Gameplay/Ascender/Sounds/Ascender_Destroyed_Cue.Ascender_Destroyed_Cue" (Transformation|GetActorLocation MyTarget))
  (Transformation|SetWorldRotation (Utilities|Array|Get(acopy) ParticleComponents 1) (Math|Rotator|CombineRotators (Transformation|GetWorldRotation (Variables|/Ascender/Gameplay/Ascender/B_Athena_Zipline_Ascender.B_Athena_Zipline_Ascender_C|GetTop (Utilities|Casting|CastToB_Athena_Zipline_Ascender MyTarget))) (Math|Vector|RotatorfromAxisandAngle (Math|Vector|GetUpVector (Transformation|GetWorldRotation (Variables|/Ascender/Gameplay/Ascender/B_Athena_Zipline_Ascender.B_Athena_Zipline_Ascender_C|GetTop (Utilities|Casting|CastToB_Athena_Zipline_Ascender MyTarget)))) -1.8e+02)) false false)
  (return))
